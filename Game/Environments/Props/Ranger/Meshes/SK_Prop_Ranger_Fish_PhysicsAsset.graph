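# Physics asset for the Ranger fish prop skeletal mesh, giving it a simple collision and physics shape.
# A single capsule body on the root bone, slightly offset and rotated, approximates the whole mesh.
# It defines no constraints, so there is no jointed ragdoll and the physics representation is one rigid shape.
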PhysicsAsset /Game/Environments/Props/Ranger/Meshes/SK_Prop_Ranger_Fish_PhysicsAsset
bodies {root [{shapeType=Capsule, length=67.309, radius=40.9725, center=<0.2568 2.3316 2.236>, rotation=<rot 0 90.0127 -89.1716>}]}
constraints []
